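; Blueprint for a cannon vehicle spawner device, apparently a cannon-specific child of a shared vehicle spawner parent.
; Its vehicle-spawned event and construction script only call the parent implementations, adding no logic of their own.
; All spawning behavior comes from the parent class.
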(event EventOnVehicleSpawned
  (params (Spawned%20Vehicle :type Fort%20Athena%20Vehicle%20Object%20Reference))
  (|CallParentFunction Spawned%20Vehicle))

(fn ConstructionScript
  (|CallParentFunction))
