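; Cannon spawner V2 device Blueprint in the vehicle spawner set, which spawns cannon vehicles.
; The vehicle-spawned event and construction script only call their parent versions, adding no logic of their own.
; A getter and setter read and write the Supports Wraps option among the basic user options.
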(event EventOnVehicleSpawned
  (params (Spawned%20Vehicle :type Fort%20Athena%20Vehicle%20Object%20Reference))
  (|CallParentFunction Spawned%20Vehicle))

(fn GetSupportsWraps
  (return (Variables|UserOptions|Basic|GetSupportsWraps)))

(fn SetSupportsWraps
  (params (Supports%20Wraps :type Boolean))
  (Variables|UserOptions|Basic|SetSupportsWraps Supports%20Wraps))

(fn ConstructionScript
  (|CallParentFunction))
